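# Physics asset for the HighTower Repulsor Blast gadget mesh, giving its arm-mounted skeleton simple collision and joint limits.
# Four mirrored capsule bodies sit on the left and right hand and upper arm bones, with the upper arm capsules offset well along the bone.
# Each hand is constrained directly to its same-side upper arm with uniform swing and twist limit angles, with no forearm body between them.
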PhysicsAsset /HighTower/Items/Tomato/Repulsors/Assets/Mesh/HighTower_Repulsor_Blast_Gadget_PhysicsAsset
bodies {
 hand_l [{shapeType=Capsule, length=10.7479, radius=6.8381, center=<6.6661 -1.9506 0.1072>, rotation=<rot 0 79.6362 -88.2833>}]
 hand_r [{shapeType=Capsule, length=10.7479, radius=6.8381, center=<-6.666 1.9507 -0.1072>, rotation=<rot 0 79.6362 -88.2833>}]
 upperarm_l [{shapeType=Capsule, length=11.5093, radius=7.0176, center=<34.1224 3.968 0.205>, rotation=<rot 0 108.2695 -85.0347>}]
 upperarm_r [{shapeType=Capsule, length=11.5093, radius=7.0176, center=<-34.1224 -3.968 -0.2049>, rotation=<rot 0 108.2694 -85.0346>}]
}
constraints [
 {bone1Name=hand_l, bone2Name=upperarm_l, swing1LimitDegrees=45, swing2LimitDegrees=45, twistLimitDegrees=45}
 {bone1Name=hand_r, bone2Name=upperarm_r, swing1LimitDegrees=45, swing2LimitDegrees=45, twistLimitDegrees=45}
]
